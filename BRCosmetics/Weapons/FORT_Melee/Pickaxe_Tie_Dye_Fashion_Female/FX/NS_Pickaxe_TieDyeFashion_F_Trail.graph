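# Niagara swing trail for the female Tie Dye Fashion pickaxe, layering animation-driven ribbons with small sprite bits.
# Trail_bits and Trail_soft001 use AnimTrail modules, blend three user TieDye colors by normalized execution index, and narrow and erode with age.
# star_bits spawns sprites by distance moved, sampled from the user skeletal mesh, inheriting velocity, scattered by curl noise and tinted with the first color.
# Trail_bright, an HDR color-curve ribbon, is disabled.
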NiagaraSystem /BRCosmetics/Weapons/FORT_Melee/Pickaxe_Tie_Dye_Fashion_Female/FX/NS_Pickaxe_TieDyeFashion_F_Trail
dependencies {
 usedDataInterfaces [&NiagaraDataInterfaceCurve, &NiagaraDataInterfaceColorCurve]
 usedDynamicInputs []
 usedModules [&SystemState]
}
emitters {
 Trail_bits {
  emitterUpdateScript [
   &EmitterState(Life Cycle Mode=::NewEnumerator1, Loop Behavior=::NewEnumerator1, Loop Delay=0.05, Loop Duration=0.2, UseLoopDelay=true)
   &FN_SpawnAnimTrail_v2()
  ]
  particleSpawnScript [
   /Game/Effects/Niagara/Modules/AnimTrail/FN_InitializeAnimTrail/FN_InitializeAnimTrailBoneIndecies(Second Bone or Socket Index=1)
   /Game/Effects/Niagara/Modules/AnimTrail/FN_InitializeAnimTrail/FN_InitializeAnimTrail(AnimTrailWidth=1, Color=<rgba 2 2 2 2>, Lifetime=0.2)
  ]
  particleUpdateScript [
   &ScaleRibbonWidth(Ribbon Width Scale=&FloatFromCurve(FloatCurve=DataInterface() {propertyValues {Curve {keys [
         {arriveTangent=-1.2579, interpMode=RCIM_Cubic, time=0, value=1}
         {arriveTangent=-1.279, interpMode=RCIM_Cubic, leaveTangent=-1.279, tangentMode=RCTM_User, time=1, value=0.4}
        ]}}}))
   &self:TriColorRamp(
    ColorA:LinearColor=$User.TieDye_Color_1
    ColorB:LinearColor=$User.TieDye_Color_2
    ColorC:LinearColor=$User.TieDye_Color_3
    InputW:NiagaraFloat=&ReturnNormalizedExecIndex()
   ) {module=TriColorRamp}
   &ScaleColor(
    Scale Alpha=&FloatFromCurve(FloatCurve=DataInterface() {propertyValues {Curve {keys [{tangentMode=RCTM_User, time=0, value=1}, {tangentMode=RCTM_User, time=1, value=0.3}]}}})
    Scale RGB=&MakeVectorFromLinearColorRGB(Color=&ColorFromCurve(ColorCurve=DataInterface() {propertyValues {
        AlphaCurve {keys [<at 0 1 RCIM_Cubic>]}
        BlueCurve {keys [
          <at 0.0682 0.2938>
          <at 0.4042 0.8683>
          {arriveTangent=0.5567, interpMode=RCIM_Cubic, leaveTangent=0.5567, time=0.7874, value=0.9148}
          <at 1 1.2 RCIM_Cubic>
         ]}
        GreenCurve {keys [
          <at 0.0682 1.0009>
          <at 0.4042 0.4511>
          {arriveTangent=0.2281, interpMode=RCIM_Cubic, leaveTangent=0.2281, time=0.7874, value=1.5}
          <at 1 0.5871 RCIM_Cubic>
         ]}
        RedCurve {keys [
          <at 0.0682 1.2>
          <at 0.4042 1.2>
          {arriveTangent=-1.9092, interpMode=RCIM_Cubic, leaveTangent=-1.9092, time=0.7874, value=0}
          <at 1 0.0625 RCIM_Cubic>
         ]}
       }}))
    ScaleA=true
    ScaleRGB=false
   ) {enabled=false, module=ScaleColor002}
   &FN_UpdateAnimTrail_v2()
   &FN_RibbonMaterial(
    Color=$Particles.Color
    Opacity Erosion:NiagaraFloat=&RemapRange(Input Value To Remap=$Particles.NormalizedAge, Input Min=0, Input Max=1, Output Min=0, Output Max=0.6)
    Scale Alpha=0.75
   )
  ]
  rendererClasses [&NiagaraRibbonRendererProperties]
  renderers [{rendererClass=&NiagaraRibbonRendererProperties}]
 }
 Trail_bright {
  bEnabled=false
  emitterUpdateScript [
   &EmitterState(Life Cycle Mode=::NewEnumerator1, Loop Behavior=::NewEnumerator1, Loop Delay=0.05, Loop Duration=0.2, UseLoopDelay=true)
   &FN_SpawnAnimTrail_v2()
  ]
  particleSpawnScript [
   &FN_InitializeAnimTrailBoneIndecies(Second Bone or Socket Index=1)
   &FN_InitializeAnimTrail(AnimTrailWidth=0.6, Color=<rgba 0 46.4696 50 1>, Lifetime=0.16)
  ]
  particleUpdateScript [
   &ScaleRibbonWidth(Ribbon Width Scale=&FloatFromCurve(FloatCurve=DataInterface() {propertyValues {Curve {keys [
         {arriveTangent=-1.2579, interpMode=RCIM_Cubic, time=0, value=1}
         {arriveTangent=-1.2665, interpMode=RCIM_Cubic, leaveTangent=-1.2665, tangentMode=RCTM_User, time=1, value=0.45}
        ]}}}))
   &ScaleColor(
    Scale Alpha=&FloatFromCurve(FloatCurve=DataInterface() {propertyValues {Curve {keys [
         {interpMode=RCIM_Cubic, leaveTangent=-0.9914, tangentMode=RCTM_Break, time=0, value=1}
         {interpMode=RCIM_Cubic, tangentMode=RCTM_Break, time=1, value=0.25}
        ]}}})
    ScaleA=true
    ScaleRGB=true
   ) {enabled=false, module=ScaleColor003}
   &ScaleColor(
    Scale Alpha=&FloatFromCurve(FloatCurve=DataInterface() {propertyValues {Curve {keys [
         {interpMode=RCIM_Cubic, leaveTangent=-0.9897, tangentMode=RCTM_Break, time=0, value=1}
         {arriveTangent=-0.2139, interpMode=RCIM_Cubic, tangentMode=RCTM_Break, time=1, value=0.3}
        ]}}})
    Scale RGB=&MakeVectorFromLinearColorRGB(Color=&ColorFromCurve(ColorCurve=DataInterface() {propertyValues {
        AlphaCurve {keys [<at 0 1 RCIM_Cubic>]}
        BlueCurve {keys [<at 0 91.4774 RCIM_Cubic>,<at 0.5494 76.789>,<at 1 150 RCIM_Cubic>]}
        GreenCurve {keys [<at 0 150 RCIM_Cubic>,<at 0.5494 30.599>,<at 1 73.3839 RCIM_Cubic>]}
        RedCurve {keys [<at 0 0 RCIM_Cubic>,<at 0.5494 125>,<at 1 7.8125 RCIM_Cubic>]}
       }}))
    ScaleA=true
    ScaleRGB=true
   ) {module=ScaleColor002}
   &ScaleRibbonWidth(Ribbon Width Scale=&FloatFromCurve(FloatCurve=DataInterface() {propertyValues {Curve {keys [<at 0 1 RCIM_Cubic>, {arriveTangent=-0.3414, interpMode=RCIM_Cubic, leaveTangent=-0.3414, tangentMode=RCTM_User, time=1, value=0.875}]}}})) {enabled=false, module=ScaleRibbonWidth001}
   &FN_UpdateAnimTrail_v2()
  ]
  rendererClasses [&NiagaraRibbonRendererProperties]
  renderers [{rendererClass=&NiagaraRibbonRendererProperties}]
 }
 Trail_soft001 {
  emitterUpdateScript [
   &EmitterState(Life Cycle Mode=::NewEnumerator1, Loop Behavior=::NewEnumerator1, Loop Delay=0.05, Loop Duration=0.2, UseLoopDelay=true)
   &FN_SpawnAnimTrail_v2()
  ]
  particleSpawnScript [
   &FN_InitializeAnimTrailBoneIndecies(Second Bone or Socket Index=1)
   &FN_InitializeAnimTrail(AnimTrailWidth=1, Color=<rgba 2 2 2 2>, Lifetime=0.2)
  ]
  particleUpdateScript [
   &ScaleRibbonWidth(Ribbon Width Scale=&FloatFromCurve(FloatCurve=DataInterface() {propertyValues {Curve {keys [
         {arriveTangent=-1.2579, interpMode=RCIM_Cubic, time=0, value=1}
         {arriveTangent=-1.279, interpMode=RCIM_Cubic, leaveTangent=-1.279, tangentMode=RCTM_User, time=1, value=0.4}
        ]}}}))
   &self:TriColorRamp(
    ColorA:LinearColor=$User.TieDye_Color_1
    ColorB:LinearColor=$User.TieDye_Color_2
    ColorC:LinearColor=$User.TieDye_Color_3
    InputW:NiagaraFloat=&ReturnNormalizedExecIndex()
   ) {module=TriColorRamp}
   &ScaleColor(
    Scale Alpha=&FloatFromCurve(FloatCurve=DataInterface() {propertyValues {Curve {keys [{tangentMode=RCTM_User, time=0, value=1}, {tangentMode=RCTM_User, time=1, value=0.3}]}}})
    Scale RGB=&MakeVectorFromLinearColorRGB(Color=&ColorFromCurve(ColorCurve=DataInterface() {propertyValues {
        AlphaCurve {keys [<at 0 1 RCIM_Cubic>]}
        BlueCurve {keys [
          <at 0.0682 0.2938>
          <at 0.4042 0.8683>
          {arriveTangent=0.5567, interpMode=RCIM_Cubic, leaveTangent=0.5567, time=0.7874, value=0.9148}
          <at 1 1.2 RCIM_Cubic>
         ]}
        GreenCurve {keys [
          <at 0.0682 1.0009>
          <at 0.4042 0.4511>
          {arriveTangent=0.2281, interpMode=RCIM_Cubic, leaveTangent=0.2281, time=0.7874, value=1.5}
          <at 1 0.5871 RCIM_Cubic>
         ]}
        RedCurve {keys [
          <at 0.0682 1.2>
          <at 0.4042 1.2>
          {arriveTangent=-1.9092, interpMode=RCIM_Cubic, leaveTangent=-1.9092, time=0.7874, value=0}
          <at 1 0.0625 RCIM_Cubic>
         ]}
       }}))
    ScaleA=true
    ScaleRGB=false
   ) {enabled=false, module=ScaleColor002}
   &FN_UpdateAnimTrail_v2()
   &FN_RibbonMaterial(
    Color=$Particles.Color
    Opacity Erosion:NiagaraFloat=&RemapRange(Input Value To Remap=$Particles.NormalizedAge, Input Min=0, Input Max=1, Output Min=0, Output Max=0.6)
    Scale Alpha=0.75
   )
  ]
  rendererClasses [&NiagaraRibbonRendererProperties]
  renderers [{rendererClass=&NiagaraRibbonRendererProperties}]
 }
 star_bits {
  emitterUpdateScript [
   &EmitterState()
   &SpawnPerUnit(Module.Use Max Movement Threshold=true, Module.Use Movement Tolerance=false, Use Spawn Probability=false)
  ]
  particleSpawnScript [
   &InitializeParticle(
    Color Mode=::NewEnumerator0
    Lifetime=&RandomRangeFloat(Minimum=0.1, Maximum=0.65)
    Sprite Size Mode=::NewEnumerator1
    Uniform Sprite Size Max=1
    Uniform Sprite Size Min=0.3
    UsePositionOffset=false
    Write Material Random=true
   )
   &SkeletalMeshLocation(
    MaskSpawnGroup=false
    Mesh Sampling Type=::NewEnumerator5
    OverrideRandomSeed=false
    Position Sampling=::NewEnumerator1
    Skeletal Mesh=$User.SkeletalMesh
   )
   &InheritVelocity(Inherited Velocity Speed Limit=125, Speed Limit=true, UseSpeedThreshold=false)
  ]
  particleUpdateScript [
   &ParticleState()
   &CurlNoiseForce(Noise Strength=800, Pan Noise=false)
   &SolveForcesAndVelocity(Clamp Velocity=false, Limit Acceleration=false)
   &Color(Color=$User.TieDye_Color_1)
   &ScaleSpriteSize(
    Uniform Curve Scale=3
    Uniform Curve Sprite Scale=DataInterface() {propertyValues {Curve {keys [{tangentMode=RCTM_User, time=0.5, value=1}, {tangentMode=RCTM_User, time=1, value=0}]}}}
   )
  ]
  rendererClasses [&NiagaraSpriteRendererProperties]
  renderers [{rendererClass=&NiagaraSpriteRendererProperties}]
 }
}
summary {
 systemName=NS_Pickaxe_TieDyeFashion_F_Trail
 emitters [
  {emitterName=Trail_bits, rendererClasses [&NiagaraRibbonRendererProperties]}
  {bEnabled=false, emitterName=Trail_bright, rendererClasses [&NiagaraRibbonRendererProperties]}
  {emitterName=star_bits, rendererClasses [&NiagaraSpriteRendererProperties]}
  {emitterName=Trail_soft001, rendererClasses [&NiagaraRibbonRendererProperties]}
 ]
 userVariables [
  User.RibbonIDOffset {type=NiagaraInt32}
  User.SkeletalMesh {
   type=NiagaraDataInterfaceSkeletalMesh
   defaultValue {
    struct=NiagaraExt_VariableValue_DataInterface
    value {dataInterface=&self:NiagaraDataInterfaceSkeletalMesh_0}
   }
  }
  User.TieDye_Color_1 {type=LinearColor, description="", defaultValue {struct=LinearColor, value=<rgba 1 0 0 1>}}
  User.TieDye_Color_2 {type=LinearColor, description="", defaultValue {struct=LinearColor, value=<rgba 0.1146 1 0.216 1>}}
  User.TieDye_Color_3 {type=LinearColor, description="", defaultValue {struct=LinearColor, value=<rgba 0 0.0826 1 1>}}
 ]
}
