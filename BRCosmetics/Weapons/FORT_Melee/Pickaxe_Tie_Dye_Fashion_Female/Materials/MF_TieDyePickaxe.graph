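# Material function for the Tie Dye Fashion pickaxe that recolors the base color of an incoming material with a tie-dye pattern.
# Three chained lerps layer four color parameters using the RGB channels of a pattern mask, which defaults to the matching character's body ombre texture and uses a shared wrap sampler.
# The Mask input blends the result into BaseColor through Get/SetMaterialAttributes, leaving other attributes untouched.
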MaterialFunction /BRCosmetics/Weapons/FORT_Melee/Pickaxe_Tie_Dye_Fashion_Female/Materials/MF_TieDyePickaxe
graph {nodes [
  n1 = GetMaterialAttributes(None=FunctionInput() {inputName=Material, inputType=FunctionInput_MaterialAttributes, sortPriority=10}) {attributeGetTypes [#1]}
  n2 = NamedRerouteDeclaration(None=@n1.BaseColor) {name=BaseColor, nodeColor=<rgba 0.7813 0 1 1>}
  n3 = NamedRerouteUsage()
  n4 = FunctionInput() {inputName=Mask, sortPriority=100}
  FunctionOutput(None=SetMaterialAttributes!(MaterialAttributes=@n1.MaterialAttributes, BaseColor=LinearInterpolate(A=@n2, B=@n3, Alpha=@n4) {constAlpha=1}) {attributeSetTypes [#1]})
  n5 = TextureSampleParameter2D() {
   parameterName=TieDye_Pattern_Mask
   samplerSource=SSM_Wrap_WorldGroupSettings
   samplerType=SAMPLERTYPE_Masks
   texture=/Game/Characters/Player/Female/Medium/Bodies/F_MED_Tie_Dye_Fashion/Textures/T_F_MED_Tie_Dye_Fashion_Body_OmbreMask
  }
  n6 = LinearInterpolate(
   A=VectorParameter() {parameterName=TieDye_Color_Base, defaultValue=<rgba 0.4853 0.6917 1 0>}.RGB
   B=VectorParameter() {parameterName=TieDye_Color_1, defaultValue=<rgba 0.0718 0.6979 0 0>}.RGB
   Alpha=@n5.R
  )
  n7 = VectorParameter() {parameterName=TieDye_Color_2, defaultValue=<rgba 0 0.133 0.849 0>}
  NamedRerouteDeclaration(None=LinearInterpolate(
    A=LinearInterpolate(A=@n6, B=@n7.RGB, Alpha=@n5.G)
    B=VectorParameter() {parameterName=TieDye_Color_3, defaultValue=<rgba 1 0 0.0338 0>}.RGB
    Alpha=@n5.B
   )) {name=TieDyeColor, nodeColor=<rgba 0.0078 0 1 1>}
 ]}
